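# Control Rig for the SteedShimmer ShireHelmet female face accessory, posing its collar and ponytail bones.
# Construction imports the skeleton, adds FK controls and builds a base_hair control; forward solve sets FK, then constrains base_collar across head, neck and spine_05.
# HairNoDYNSetup extracts base_hair swing against that control, remaps it through curves and counter-scaled rotations onto dyn_ponytailM bones.
# NoRBAN weights these rotations, apparently as a fallback without rigid-body dynamics.
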ControlRigBlueprint /BRCosmetics/Characters/Player/Female/Medium/Bodies/F_MED_SteedShimmer_ShireHelmet/Meshes/Parts/F_MED_SteedShimmer_ShireHelmet_FaceAcc_CtrlRig
graphs [
 New Function_ContainedGraph {nodes [
   RigVMFunctionEntryNode()
   > RigVMFunction_Sequence = RigVMAggregateNode()
   RigVMFunctionReturnNode(ExecuteContext=@RigVMFunction_Sequence.C)
   n1 = &RigUnit_GetRelativeTransformForItem(Child=(Type=Bone,Name="base_hair"), Parent=(Type=Control,Name="base_hair_ctrl"), bChildInitial=False, bParentInitial=False)
   RigVMVariableNode(
    Value=&RigVMFunction_MathQuaternionToEuler(
     RotationOrder=ZYX
     Value=&RigVMFunction_MathQuaternionSwingTwist(TwistAxis=(X=1,Y=0,Z=0), Input=@n1.RelativeTransform.Rotation).Swing
    ).Result
    ExecuteContext=@RigVMFunction_Sequence.A
   ) {variable=headSwing_rot}
   RigVMFunction_MathBoolToFloat = &RigVMFunction_MathBoolToFloat(Value=RigVMVariableNode() {variable=NoRBAN}.Value)
   n2 = RigVMVariableNode() {variable=headSwing_rot}
   n3 = &RigVMFunction_AnimEvalRichCurve(
    Curve="(EditorCurveData=(Keys=((InterpMode=RCIM_Cubic,TangentMode=RCTM_User,ArriveTangent=0.2796,LeaveTangent=0.2796),(InterpMode=RCIM_Cubic,TangentMode=RCTM_User,Time=0.5061,Value=0.359,ArriveTangent=1.2761,LeaveTangent=1.2761),(InterpMode=RCIM_Cubic,TangentMode ...(truncated)"
    SourceMaximum=-140
    SourceMinimum=0
    TargetMaximum=-44
    TargetMinimum=0
    Value=@n2.Value.Y
   )
   n4 = &RigVMFunction_AnimEvalRichCurve(
    Curve="(EditorCurveData=(Keys=((InterpMode=RCIM_Cubic,TangentMode=RCTM_User,ArriveTangent=0.2796,LeaveTangent=0.2796),(InterpMode=RCIM_Cubic,TangentMode=RCTM_User,Time=0.4829,Value=0.6231,ArriveTangent=1.0187,LeaveTangent=1.0187),(InterpMode=RCIM_Cubic,TangentMod ...(truncated)"
    SourceMaximum=-140
    SourceMinimum=0
    TargetMaximum=-44
    TargetMinimum=0
    Value=@n2.Value.Z
   )
   n5 = &RigVMFunction_AlphaInterp(
    InterpSpeedDecreasing=10
    InterpSpeedIncreasing=10
    Scale=1
    bClampResult=false
    bInterpResult=false
    bMapRange=true
    Value=&RigVMFunction_MathDoubleAdd(A=@n3.Result, B=@n4.Result).Result
   )
   n6 = &RigVMFunction_MathQuaternionFromEuler(Euler=(X=0,Y=0,Z=0), Euler.Y=&RigVMFunction_MathDoubleMul(B=0.45, A=@n5.Result).Result)
   &RigUnit_ModifyTransforms(
    ItemToModify="((Item=(Type=Bone,Name=\"dyn_ponytailM_01\"),Transform=(Rotation=(X=0,Y=0,Z=0,W=1),Translation=(X=0,Y=0,Z=0),Scale3D=(X=1,Y=1,Z=1))),(Item=(Type=Bone,Name=\"dyn_ponytailM_03\"),Transform=(Rotation=(X=0,Y=0,Z=0,W=1),Translation=(X=0,Y=0,Z=0),Scale3D=(X=1,Y=1,Z= ...(truncated)"
    WeightMinimum=0
    Weight=@RigVMFunction_MathBoolToFloat.Result
    ItemToModify.0.Transform.Rotation=@n6.Result
    ItemToModify.1.Transform.Rotation=@n6.Result
    ItemToModify.2.Transform.Rotation=@n6.Result
    ExecutePin=@RigVMFunction_Sequence.B
   )
   n7 = &RigVMFunction_MathQuaternionFromEuler(Euler=(X=0,Y=0,Z=0), Euler.Y=&RigVMFunction_MathDoubleMul(B=0.65, A=@n5.Result).Result)
   > &RigUnit_ModifyTransforms(
    ItemToModify="((Item=(Type=Bone,Name=\"dyn_ponytailM_05\"),Transform=(Rotation=(X=0,Y=0,Z=0,W=1),Translation=(X=0,Y=0,Z=0),Scale3D=(X=1,Y=1,Z=1))),(Item=(Type=Bone,Name=\"dyn_ponytailM_06\"),Transform=(Rotation=(X=0,Y=0,Z=0,W=1),Translation=(X=0,Y=0,Z=0),Scale3D=(X=1,Y=1,Z= ...(truncated)"
    WeightMinimum=0
    Weight=@RigVMFunction_MathBoolToFloat.Result
    ItemToModify.0.Transform.Rotation=@n7.Result
    ItemToModify.1.Transform.Rotation=@n7.Result
    ItemToModify.2.Transform.Rotation=@n7.Result
   )
   n8 = &RigVMFunction_MathBoolToFloat(Value=RigVMVariableNode() {variable=NoRBAN}.Value)
   n9 = RigVMVariableNode() {variable=headSwing_rot}
   n10 = &RigVMFunction_AnimEvalRichCurve(
    Curve="(EditorCurveData=(Keys=((InterpMode=RCIM_Cubic,TangentMode=RCTM_User,ArriveTangent=0.2796,LeaveTangent=0.2796),(InterpMode=RCIM_Cubic,TangentMode=RCTM_User,Time=0.5061,Value=0.359,ArriveTangent=1.2761,LeaveTangent=1.2761),(InterpMode=RCIM_Cubic,TangentMode ...(truncated)"
    SourceMaximum=140
    SourceMinimum=0
    TargetMaximum=-44
    TargetMinimum=0
    Value=@n9.Value.Y
   )
   n11 = &RigVMFunction_AnimEvalRichCurve(
    Curve="(EditorCurveData=(Keys=((InterpMode=RCIM_Cubic,TangentMode=RCTM_User,ArriveTangent=0.2796,LeaveTangent=0.2796),(InterpMode=RCIM_Cubic,TangentMode=RCTM_User,Time=0.4829,Value=0.6231,ArriveTangent=1.0187,LeaveTangent=1.0187),(InterpMode=RCIM_Cubic,TangentMod ...(truncated)"
    SourceMaximum=140
    SourceMinimum=0
    TargetMaximum=-44
    TargetMinimum=0
    Value=@n9.Value.Z
   )
   n12 = &RigVMFunction_AlphaInterp(
    InterpSpeedDecreasing=10
    InterpSpeedIncreasing=10
    Scale=1
    bClampResult=false
    bInterpResult=false
    bMapRange=true
    Value=&RigVMFunction_MathDoubleAdd(A=@n10.Result, B=@n11.Result).Result
   )
   n13 = &RigVMFunction_MathQuaternionFromEuler(Euler=(X=0,Y=0,Z=0), Euler.Y=&RigVMFunction_MathDoubleMul(B=-1, A=@n12.Result).Result)
   > &RigUnit_ModifyTransforms(
    ItemToModify="((Item=(Type=Bone,Name=\"dyn_ponytailM_01\"),Transform=(Rotation=(X=0,Y=0,Z=0,W=1),Translation=(X=0,Y=0,Z=0),Scale3D=(X=1,Y=1,Z=1))),(Item=(Type=Bone,Name=\"dyn_ponytailM_03\"),Transform=(Rotation=(X=0,Y=0,Z=0,W=1),Translation=(X=0,Y=0,Z=0),Scale3D=(X=1,Y=1,Z= ...(truncated)"
    WeightMinimum=0
    Weight=@n8.Result
    ItemToModify.0.Transform.Rotation=@n13.Result
    ItemToModify.1.Transform.Rotation=@n13.Result
    ItemToModify.2.Transform.Rotation=@n13.Result
   )
   n14 = &RigVMFunction_MathQuaternionFromEuler(Euler=(X=0,Y=0,Z=0), Euler.Y=&RigVMFunction_MathDoubleMul(B=-0.5, A=@n12.Result).Result)
   > &RigUnit_ModifyTransforms(
    ItemToModify="((Item=(Type=Bone,Name=\"dyn_ponytailM_05\"),Transform=(Rotation=(X=0,Y=0,Z=0,W=1),Translation=(X=0,Y=0,Z=0),Scale3D=(X=1,Y=1,Z=1))),(Item=(Type=Bone,Name=\"dyn_ponytailM_06\"),Transform=(Rotation=(X=0,Y=0,Z=0,W=1),Translation=(X=0,Y=0,Z=0),Scale3D=(X=1,Y=1,Z= ...(truncated)"
    WeightMinimum=0
    Weight=@n8.Result
    ItemToModify.0.Transform.Rotation=@n14.Result
    ItemToModify.1.Transform.Rotation=@n14.Result
    ItemToModify.2.Transform.Rotation=@n14.Result
   )
  ]}
 RigVMFunctionLibrary {nodes [HairNoDYNSetup = RigVMCollapseNode(ExecuteContext=())]}
 RigVMFunction_Sequence_1_ContainedGraph {nodes [
   n1 = RigVMFunctionEntryNode()
   RigVMFunction_Sequence = &RigVMFunction_Sequence(ExecuteContext=@n1.ExecuteContext)
   n2 = &RigVMFunction_Sequence(ExecuteContext=@RigVMFunction_Sequence.B)
   RigVMFunctionReturnNode(A=@RigVMFunction_Sequence.A, B=@n2.A, C=@n2.B)
  ]}
 RigVMModel {nodes [
   n1 = &RigUnit_PrepareForExecution()
   &RigUnit_HierarchyImportFromSkeleton(NameSpace=None, bIncludeCurves=False, bIncludeMeshSockets=False, bIncludeVirtualBones=True, ExecutePin=@n1.ExecutePin)
   > addFKCtrl = RigVMFunctionReferenceNode(
    Custom_FKCtrl=()
    Options_FKCtrl="(pelvis_1_#1=False,spine_3_#2=true,clavicle_5_#3=False,arm_7_#4=False,leg_9_#5=False,head_11_46DC15EF465CD ...(truncated)"
    Shape_Settings="(pelvis_2_#6=(bVisible=True,Name=\"Hexagon_Thin\",Color=(R=0.0117,G=0.8281,B=1,A=1),Transform=(Rotation=(X=0,Y=-0.7071,Z=0,W=0.7071),Translation=(X=2,Y=0,Z=0),Scale3D=(X=5,Y=5,Z=7))),spine_4_#7=(bVi ...(truncated)"
   )
   > Control Stack at Item = RigVMFunctionReferenceNode(
    Add Bottom Null=False
    Add Null=False
    Add Secondary Control=False
    Control Suffix=_ctrl
    Item=(Name="base_hair",Type=Bone)
    Null Suffix=_null
    Parent=(Name="spine_05",Type=Bone)
    Secondary Settings="(Shape=(bVisible=True,Name=\"Default\",Color=(R=1,G=0,B=0,A=1),Transform=(Rotation=(X=0,Y=0,Z=0,W=1),Translation=(X=0,Y=0,Z=0),Scale3D=(X=1,Y=1,Z=1))),Proxy=(bIsProxy=False,ShapeVisibility=BasedOnSelection),DisplayName=\"None\",InitialSpace=LocalSpace,bUsePref ...(truncated)"
    Settings="(Shape=(bVisible=True,Name=\"Default\",Color=(R=0,G=0,B=1,A=1),Transform=(Rotation=(X=0,Y=0,Z=0,W=1),Translation=(X=0,Y=0,Z=0),Scale3D=(X=1,Y=1,Z=1))),Proxy=(bIsProxy=False,ShapeVisibility=BasedOnSelection),DisplayName=\"None\",InitialSpace=LocalSpace,bUsePref ...(truncated)"
    Tag Controls=true
    Zero Controls=true
   )
   RigUnit_BeginExecution = &RigUnit_BeginExecution()
   setFKCtrl = RigVMFunctionReferenceNode(setFK_Ctrl=false, ExecuteContext=@RigUnit_BeginExecution.ExecutePin)
   > &RigUnit_ParentConstraint(
    AdvancedSettings=(RotationOrderForFilter=XZY)
    Child=(Type=Bone,Name="base_collar")
    Parents=((Item=(Type=Bone,Name="head"),Weight=0.0405),(Item=(Type=Bone,Name="neck_01"),Weight=0.2827),(Item=(Type=Bone,Name="neck_02"),Weight=0.2828),(Item=(Type=Bone,Name="spine_05"),Weight=0.394))
    Weight=1
   )
   > HairNoDYNSetup = RigVMFunctionReferenceNode()
  ]}
]
variables [
 NoRBAN {type=bool, default=False}
 headSwing_rot {type=FVector, default=(X=0,Y=0,Z=0)}
]
